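# Post-copy AnimBP for the large male SnakeCrest body cosmetic, adding secondary motion to its extra armor, pipe, mirror and carabiner bones.
# The pose is copied from the parent mesh, then thigh-driven bone drivers, spine/clavicle constraints, twist-weighted pipe offsets and AnimDynamics carabiners adjust it.
# A final Control Rig receives parachute, targeting and combined skydive/emote/melee alphas.
# The EventGraph eases those alphas and derives pipe weights from thigh twist relative to pelvis.
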AnimBlueprint /BRCosmetics/Characters/Player/Male/Large/Bodies/M_LRG_SnakeCrest/Meshes/M_LRG_SnakeCrest_AnimBP
target_skeleton=&/M_LRG_SnakeCrest_Skeleton
graphs [
 AnimGraph {nodes [
   n1 = "Transform (Modify) Bone - Bone: upperarm_l"(
    ComponentPose=Local To Component(LocalPose=Copy Pose From Mesh() {bUseAttachedParent=true})
    Alpha=1
    Translation="0,0,0"
    Rotation="0,0,0"
    Scale="1,1,1"
   ) {rotationMode=BMM_Replace, rotationSpace=BCS_BoneSpace, alphaBoolBlend {blendOption=Linear}, boneToModify {boneName=upperarm_l}}
   n2 = "Transform (Modify) Bone - Bone: upperarm_l"(ComponentPose=@n1, Alpha=1, Translation="0,0,0", Rotation="0,0,0", Scale="1,1,1") {rotationMode=BMM_Additive, rotationSpace=BCS_BoneSpace, alphaBoolBlend {blendOption=Linear}, boneToModify {boneName=upperarm_l}}
   n3 = "Transform (Modify) Bone - Bone: extra_pipe_l"(
    ComponentPose=@n2
    Alpha=Get extra_thigh_pipe_l() -> (extra_thigh_pipe_l:double)
    Translation="3.265,-6.600,-11.149"
    Rotation="-37.293,-58.970,71.537"
    Scale="1,1,1"
   ) {
    rotationMode=BMM_Replace
    rotationSpace=BCS_BoneSpace
    translationMode=BMM_Replace
    translationSpace=BCS_BoneSpace
    alphaBoolBlend {blendOption=Linear}
    boneToModify {boneName=extra_pipe_l}
    rotation=<rot -37.293 -58.97 71.537>
    translation=<3.265 -6.6 -11.149>
   }
   n4 = "Transform (Modify) Bone - Bone: extra_pipe_r"(
    ComponentPose=@n3
    Alpha=Get extra_thigh_pipe_r() -> (extra_thigh_pipe_r:double)
    Translation="3.265,6.6,-11.149"
    Rotation="-37.293,58.97,-71.537"
    Scale="1,1,1"
   ) {
    rotationMode=BMM_Additive
    rotationSpace=BCS_BoneSpace
    translationMode=BMM_Replace
    translationSpace=BCS_BoneSpace
    alphaBoolBlend {blendOption=Linear}
    boneToModify {boneName=extra_pipe_r}
    rotation=<rot -37.293 58.97 -71.537>
    translation=<3.265 6.6 -11.149>
   }
   n5 = "extra_thigh_armor_r.rotateX = remap(thigh_r.rotateX) - Bone Driven Controller"!(ComponentPose:Component Space Pose Link Structure=@n4, Alpha:float=1) -> (Pose:Component Space Pose Link Structure) {
    is=title AnimGraphNode_BoneDrivenController
    bAffectTargetRotationX=true
    bUseRange=true
    rangeMax=-60
    rangeMin=60
    remappedMax=-60
    remappedMin=60
    sourceComponent=RotationX
    alphaBoolBlend {blendOption=Linear}
    sourceBone {boneName=thigh_r}
    targetBone {boneName=extra_thigh_armor_r}
   }
   n6 = "extra_thigh_armor_l.rotateX = remap(thigh_l.rotateX) - Bone Driven Controller"!(ComponentPose:Component Space Pose Link Structure=@n5, Alpha:float=1) -> (Pose:Component Space Pose Link Structure) {
    is=title AnimGraphNode_BoneDrivenController
    bAffectTargetRotationX=true
    bUseRange=true
    rangeMax=-60
    rangeMin=60
    remappedMax=60
    remappedMin=-60
    sourceComponent=RotationX
    alphaBoolBlend {blendOption=Linear}
    sourceBone {boneName=thigh_l}
    targetBone {boneName=extra_thigh_armor_l}
   }
   n7 = "extra_thigh_armor_r.rotateY = remap(thigh_r.rotateY) - Bone Driven Controller"!(ComponentPose:Component Space Pose Link Structure=@n6, Alpha:float=1) -> (Pose:Component Space Pose Link Structure) {
    is=title AnimGraphNode_BoneDrivenController
    bAffectTargetRotationY=true
    bUseRange=true
    rangeMax=-60
    rangeMin=60
    remappedMax=-60
    remappedMin=60
    sourceComponent=RotationY
    alphaBoolBlend {blendOption=Linear}
    sourceBone {boneName=thigh_r}
    targetBone {boneName=extra_thigh_armor_r}
   }
   n8 = "extra_thigh_armor_l.rotateY = remap(thigh_l.rotateY) - Bone Driven Controller"!(ComponentPose:Component Space Pose Link Structure=@n7, Alpha:float=1) -> (Pose:Component Space Pose Link Structure) {
    is=title AnimGraphNode_BoneDrivenController
    bAffectTargetRotationY=true
    bUseRange=true
    rangeMax=60
    rangeMin=-60
    remappedMax=-60
    remappedMin=60
    sourceComponent=RotationY
    alphaBoolBlend {blendOption=Linear}
    sourceBone {boneName=thigh_l}
    targetBone {boneName=extra_thigh_armor_l}
   }
   n9 = "Constraint - extra_mirror_base_l (PPPPPP)"!(
    ComponentPose=@n8
    Alpha=1
    ConstraintWeights_0=0.425
    ConstraintWeights_1:float=0.036
    ConstraintWeights_2:float=0.066
    ConstraintWeights_3:float=0.113
    ConstraintWeights_4:float=0.141
    ConstraintWeights_5:float=0.22
   ) -> (Pose) {
    alphaBoolBlend {blendOption=Linear}
    boneToModify {boneName=extra_mirror_base_l}
    constraintSetup [
     {targetBone {boneName=clavicle_l}}
     {targetBone {boneName=spine_01}}
     {targetBone {boneName=spine_02}}
     {targetBone {boneName=spine_03}}
     {targetBone {boneName=spine_04}}
     {targetBone {boneName=spine_05}}
    ]
    constraintWeights [1,1,1,1,1,1]
   }
   n10 = "Constraint - extra_mirror_base_r (PPPPPP)"!(
    ComponentPose=@n9
    Alpha=1
    ConstraintWeights_0=0.425
    ConstraintWeights_1:float=0.036
    ConstraintWeights_2:float=0.066
    ConstraintWeights_3:float=0.113
    ConstraintWeights_4:float=0.141
    ConstraintWeights_5:float=0.22
   ) -> (Pose) {
    alphaBoolBlend {blendOption=Linear}
    boneToModify {boneName=extra_mirror_base_r}
    constraintSetup [
     {targetBone {boneName=clavicle_r}}
     {targetBone {boneName=spine_01}}
     {targetBone {boneName=spine_02}}
     {targetBone {boneName=spine_03}}
     {targetBone {boneName=spine_04}}
     {targetBone {boneName=spine_05}}
    ]
    constraintWeights [1,1,1,1,1,1]
   }
   n11 = Anim Dynamics - Bone: dyn_carabin_01_l(ComponentPose=@n10, Alpha=1, ExternalForce="0,0,0") {
    angularBiasOverride=0.9
    angularDampingOverride=0.7
    angularSpringConstant=1500
    bGravityOverrideInSimSpace=true
    bOverrideAngularBias=true
    bOverrideAngularDamping=true
    bUsePlanarLimit=false
    gravityScale=0.35
    simulationSpace=RootRelative
    alphaBoolBlend {blendOption=Linear}
    boundBone {boneName=dyn_carabin_01_l}
    componentAppliedLinearAccClamp=<200 200 200>
    componentLinearAccScale {z=0.3}
    componentLinearVelScale {z=0.3}
    physicsBodyDefinitions [{
      sphereCollisionRadius=0
      boundBone {boneName=dyn_carabin_01_l}
      boxExtents=<10 5 5>
      constraintSetup {angularLimitsMax=<0 0 0>, angularLimitsMin=<0 0 -115>, angularTarget=<0 0 0>}
      localJointOffset=<2 0 0>
     }]
   }
   n12 = Anim Dynamics - Bone: dyn_carabin_01_r(ComponentPose=@n11, Alpha=1, ExternalForce="0,0,0") {
    angularBiasOverride=0.9
    angularDampingOverride=0.7
    angularSpringConstant=1500
    bGravityOverrideInSimSpace=true
    bOverrideAngularBias=true
    bOverrideAngularDamping=true
    bUsePlanarLimit=false
    gravityScale=0.35
    simulationSpace=RootRelative
    alphaBoolBlend {blendOption=Linear}
    boundBone {boneName=dyn_carabin_01_r}
    componentAppliedLinearAccClamp=<200 200 200>
    componentLinearAccScale {z=0.3}
    componentLinearVelScale {z=0.3}
    physicsBodyDefinitions [{
      sphereCollisionRadius=0
      boundBone {boneName=dyn_carabin_01_r}
      boxExtents=<10 5 5>
      constraintSetup {angularLimitsMax=<0 0 0>, angularLimitsMin=<0 0 -115>, angularTarget=<0 0 0>}
      localJointOffset=<2 0 0>
     }]
   }
   n13 = Control Rig!(
    Source=Component To Local(ComponentPose=@n12)
    Alpha
    parachute_alpha:double=Get parachute_alpha() -> (parachute_alpha:double)
    is_or_skydive_emote_melee_alpha:double=Get is_or_skydive_emote_melee_alpha() -> (is_or_skydive_emote_melee_alpha:double)
    targeting_alpha:double=Get targeting_alpha() -> (targeting_alpha:double)
   ) -> (Pose) {
    alphaBoolBlend {blendOption=Linear}
    controlRigAssetReference {blueprintRigClass=&/M_LRG_SnakeCrest_CtrlRig.M_LRG_SnakeCrest_CtrlRig_C}
   }
   Output Pose(Result=@n13) {layerGroup=None}
  ]}
 EventGraph {nodes [
   n1 = Event BlueprintUpdateAnimation()
   > n2 = Sequence!(execute) -> (then_0, then_1, then_2:Exec, then_3:Exec, then_4:Exec, then_5:Exec, then_6:Exec, then_7:Exec)
   > Set DeltaTimeX(DeltaTimeX:double=@n1.DeltaTimeX) -> (Output_Get:double)
   > Set Owner Skeletal(Owner Skeletal:Skeletal Mesh Component Object Reference=GetPartSkeletalMeshComponent()) -> (Output_Get:Skeletal Mesh Component Object Reference)
   n3 = Branch(execute=@n2.then_1, Condition=Get bIsParachuteOpened())
   > Set parachute_alpha(parachute_alpha:double=FInterpTo_Constant!(Current:double=Get parachute_alpha(), Target:double=1, DeltaTime:double=Get DeltaTimeX() -> (DeltaTimeX:double), InterpSpeed:double=2) -> (ReturnValue:double)) -> (Output_Get:double)
   Set parachute_alpha(execute=@n3.else, parachute_alpha=FInterpTo_Constant(Current=Get parachute_alpha(), Target=0, DeltaTime=Get DeltaTimeX(), InterpSpeed=2))
   n4 = Branch(execute=@n2.then_2, Condition=Get bIsSkyDiving())
   > Set skydive_alpha(skydive_alpha:double=FInterpTo_Constant(Current=Get skydive_alpha() -> (skydive_alpha:double), Target=1, DeltaTime=Get DeltaTimeX(), InterpSpeed=2)) -> (Output_Get:double)
   Set skydive_alpha(execute=@n4.else, skydive_alpha=FInterpTo_Constant(Current=Get skydive_alpha(), Target=0, DeltaTime=Get DeltaTimeX(), InterpSpeed=2))
   n5 = Branch(execute=@n2.then_3, Condition=Get bIsPlayingEmote())
   > Set emote_alpha(emote_alpha:double=FInterpTo_Constant(Current=Get emote_alpha() -> (emote_alpha:double), Target=1, DeltaTime=Get DeltaTimeX(), InterpSpeed=2)) -> (Output_Get:double)
   Set emote_alpha(execute=@n5.else, emote_alpha=FInterpTo_Constant(Current=Get emote_alpha(), Target=0, DeltaTime=Get DeltaTimeX(), InterpSpeed=2))
   n6 = Branch(execute=@n2.then_4, Condition=Get bIsPlayingMeleeAnim())
   > Set melee_alpha(melee_alpha:double=FInterpTo_Constant(Current=Get melee_alpha() -> (melee_alpha:double), Target=1, DeltaTime=Get DeltaTimeX(), InterpSpeed=2)) -> (Output_Get:double)
   Set melee_alpha(execute=@n6.else, melee_alpha=FInterpTo_Constant(Current=Get melee_alpha(), Target=0, DeltaTime=Get DeltaTimeX(), InterpSpeed=2))
   Set is_or_skydive_emote_melee_alpha(
    execute=@n2.then_5
    is_or_skydive_emote_melee_alpha:double="Max (Float)"!(A=Get skydive_alpha(), B=Get emote_alpha(), C:double=Get melee_alpha()) -> (ReturnValue)
   ) -> (Output_Get:double)
   Set extra_thigh_pipe_l(
    execute=@n2.then_6
    extra_thigh_pipe_l:double=GetAlphaFromReferenceTransformDelta(
     SkeletalMeshComponent=Get Owner Skeletal() -> (Owner Skeletal:Skeletal Mesh Component Object Reference)
     BoneName=pelvis
     BaseName=thigh_l
     DeltaComponentType=QuaternionTwist
     InputRange=(X=0,Y=-100)
     TwistAxis="0,0,1"
    )
   ) -> (Output_Get:double)
   > Set extra_thigh_pipe_r(extra_thigh_pipe_r:double=GetAlphaFromReferenceTransformDelta(
     SkeletalMeshComponent=Get Owner Skeletal()
     BoneName=pelvis
     BaseName=thigh_r
     DeltaComponentType=QuaternionTwist
     InputRange=(X=0,Y=-100)
     TwistAxis="0,0,1"
    )) -> (Output_Get:double)
   n7 = Branch(execute=@n2.then_7, Condition=Get bIsTargeting())
   > Set targeting_alpha(targeting_alpha:double=FInterpTo_Constant(Current=Get targeting_alpha(), Target=1, DeltaTime=Get DeltaTimeX(), InterpSpeed=6)) -> (Output_Get:double)
   Set targeting_alpha(execute=@n7.else, targeting_alpha=FInterpTo_Constant(Current=Get targeting_alpha(), Target=0, DeltaTime=Get DeltaTimeX(), InterpSpeed=3))
  ]}
]
variables [
 DeltaTimeX {type=number}
 emote_alpha {type=number}
 extra_thigh_pipe_l {type=number}
 extra_thigh_pipe_r {type=number}
 is_or_skydive_emote_melee_alpha {type=number}
 melee_alpha {type=number}
 Owner Skeletal {type=SkeletalMeshComponent}
 parachute_alpha {type=number}
 skydive_alpha {type=number}
 targeting_alpha {type=number}
]
