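# Control Rig for the SnakeCrest large male face accessory that procedurally animates a pair of left and right piston bones between spine_05 and neck_02.
# Two spherical pose readers measure head orientation relative to spine_05, both with debug drawing enabled.
# Their maximum weights a small Y offset on the left piston, while the first reader alone weights a smaller offset on the right.
# Aim constraints then point each piston and its end bone at each other.
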ControlRigBlueprint /BRCosmetics/Characters/Player/Male/Large/Bodies/M_LRG_SnakeCrest/Meshes/Parts/M_LRG_SnakeCrest_FaceAcc_CtrlRig
graphs [
 RigVMFunctionLibrary {links [], nodes []}
 RigVMModel {nodes [
   RigUnit_BeginExecution = &RigUnit_BeginExecution()
   n1 = &RigUnit_SphericalPoseReader(
    ActiveRegionScaleFactors=(PositiveWidth=0.4,NegativeWidth=0.3,PositiveHeight=2.5,NegativeHeight=2.5)
    ActiveRegionSize=0.1
    Debug=(bDrawDebug=true,DebugScale=25,DebugSegments=20,DebugThickness=0.25)
    DriverAxis=(X=0,Y=1,Z=0)
    DriverItem=(Type=Bone,Name="head")
    FalloffRegionScaleFactors=(PositiveWidth=0.2,NegativeWidth=0.2,PositiveHeight=2,NegativeHeight=2)
    FalloffSize=0.3
    OptionalParentItem=(Type=Bone,Name="spine_05")
    RotationOffset=(X=90,Y=0,Z=20)
    ExecutePin=@RigUnit_BeginExecution.ExecutePin
   )
   > n2 = &RigUnit_SphericalPoseReader(
    ActiveRegionScaleFactors=(PositiveWidth=1,NegativeWidth=1,PositiveHeight=0.4,NegativeHeight=0.4)
    ActiveRegionSize=0.2
    Debug=(bDrawDebug=true,DebugScale=25,DebugSegments=20,DebugThickness=0.25)
    DriverAxis=(X=0,Y=1,Z=0)
    DriverItem=(Type=Bone,Name="head")
    FalloffRegionScaleFactors=(PositiveWidth=0.1,NegativeWidth=0.1,PositiveHeight=0.4,NegativeHeight=0.4)
    FalloffSize=0.3
    OptionalParentItem=(Type=Bone,Name="spine_05")
    RotationOffset=(X=127,Y=10,Z=-20)
   )
   Maximum = &RigVMFunction_MathDoubleMax(A=@n1.OutputParam, B=@n2.OutputParam)
   > &RigUnit_ModifyTransforms(
    ItemToModify=((Item=(Type=Bone,Name="extra_piston_l"),Transform=(Rotation=(X=0,Y=0,Z=0,W=1),Translation=(X=0,Y=-0.35,Z=0),Scale3D=(X=1,Y=1,Z=1))))
    Weight=@Maximum.Result
   )
   > &RigUnit_ModifyTransforms(
    ItemToModify=((Item=(Type=Bone,Name="extra_piston_r"),Transform=(Rotation=(X=0,Y=0,Z=0,W=1),Translation=(X=0,Y=-0.2,Z=0),Scale3D=(X=1,Y=1,Z=1))))
    Weight=@n1.OutputParam
   )
   > &RigUnit_AimConstraintLocalSpaceOffset(
    AdvancedSettings=(DebugSettings=(bEnabled=false,Scale=10,WorldOffset=(Rotation=(X=0,Y=0,Z=0,W=1),Translation=(X=0,Y=0,Z=0),Scale3D=(X=1,Y=1,Z=1))),RotationOrderForFilter=XZY)
    Child=(Type=Bone,Name="extra_piston_l")
    Filter=(bX=false,bY=True,bZ=True)
    Parents=((Item=(Type=Bone,Name="extra_piston_end_l"),Weight=1))
    Weight=1
    WorldUp=(Target=(X=0,Y=0,Z=1),Space=(Type=None,Name="extra_piston_l"))
   )
   > &RigUnit_AimConstraintLocalSpaceOffset(
    AdvancedSettings=(DebugSettings=(bEnabled=false,Scale=10,WorldOffset=(Rotation=(X=0,Y=0,Z=0,W=1),Translation=(X=0,Y=0,Z=0),Scale3D=(X=1,Y=1,Z=1))),RotationOrderForFilter=XZY)
    Child=(Type=Bone,Name="extra_piston_r")
    Filter=(bX=false,bY=True,bZ=true)
    Parents=((Item=(Type=Bone,Name="extra_piston_end_r"),Weight=1))
    Weight=1
    WorldUp=(Target=(X=0,Y=0,Z=1),Space=(Type=None,Name="extra_piston_r"))
   )
   > &RigUnit_AimConstraintLocalSpaceOffset(
    AdvancedSettings=(DebugSettings=(Scale=10,WorldOffset=(Rotation=(X=0,Y=0,Z=0,W=1),Translation=(X=0,Y=0,Z=0),Scale3D=(X=1,Y=1,Z=1))),RotationOrderForFilter=XZY)
    Child=(Type=Bone,Name="extra_piston_end_l")
    Filter=(bX=true,bY=true,bZ=true)
    Parents=((Item=(Type=Bone,Name="extra_piston_l"),Weight=1))
    Weight=1
    WorldUp=(Target=(X=0,Y=0,Z=1),Space=(Type=None,Name="None"))
   )
   > &RigUnit_AimConstraintLocalSpaceOffset(
    AdvancedSettings=(DebugSettings=(Scale=10,WorldOffset=(Rotation=(X=0,Y=0,Z=0,W=1),Translation=(X=0,Y=0,Z=0),Scale3D=(X=1,Y=1,Z=1))),RotationOrderForFilter=XZY)
    Child=(Type=Bone,Name="extra_piston_end_r")
    Filter=(bX=true,bY=True,bZ=true)
    Parents=((Item=(Type=Bone,Name="extra_piston_r"),Weight=1))
    Weight=1
    WorldUp=(Target=(X=0,Y=0,Z=1),Space=(Type=None,Name="None"))
   )
  ]}
]
hierarchy [
 root {loc=<0 0 0>, rot=<0 0 0 1>}
 pelvis {parent=root, loc=<0 -0.7321 92.4218>, rot=<0.0127 0.707 -0.0127 -0.707>}
 spine_01 {parent=pelvis, loc=<5.6002 -0.5128 0>, rot=<0 0 0.034 -0.9994>}
 spine_02 {parent=spine_01, loc=<8.0753 0.0139 0>, rot=<0 0 0.0238 -0.9997>}
 spine_03 {parent=spine_02, loc=<7.0393 0 0>, rot=<0 0 0.0149 -0.9999>}
 spine_04 {parent=spine_03, loc=<8.1024 0 0>, rot=<0 0 -0.0132 -0.9999>}
 spine_05 {parent=spine_04, loc=<10.0146 0 0>, rot=<0 0 -0.0015 -1>}
 neck_01 {parent=spine_05, loc=<10.815 1.9196 0>, rot=<0 0 -0.1864 -0.9825>}
 neck_02 {parent=neck_01, loc=<4.7347 0.3045 0>, rot=<0 0 0.0363 -0.9993>}
 head {parent=neck_02, loc=<5.7632 0.5552 0>, rot=<0 0 0.115 -0.9934>}
 extra_piston_end_l {parent=neck_02, loc=<-1.1374 2.8507 -3.0303>, rot=<-0.0784 -0.7027 0.0784 -0.7027>}
 extra_piston_end_r {parent=neck_02, loc=<-1.1378 2.8508 3.0303>, rot=<-0.0784 -0.7027 0.0784 -0.7027>}
 extra_piston_l {parent=spine_05, loc=<4.5186 7.2607 -5.3439>, rot=<-0.019 0.1286 0.0437 -0.9906>}
 extra_piston_r {parent=spine_05, loc=<4.5187 7.2607 5.3439>, rot=<0.019 -0.1286 0.0437 -0.9906>}
]
